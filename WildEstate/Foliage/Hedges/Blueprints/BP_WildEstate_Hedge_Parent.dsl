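; Parent Blueprint for Wild Estate hedges, handling leaf color, outline merging and death effects.
; Construction applies season presets from a color data asset, writing them to Custom Primitive Data.
; With merging enabled, a transform change triggers a recursive overlap search for connected hedges, which share one position-seeded random value through a roughness data slot; otherwise each hedge derives its own.
; On death, mobile games spawn a Niagara destruction effect with FX customization.
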(event EventBeginPlay)

(event Collision|EventActorBeginOverlap
  (params (OtherActor :type Actor%20Object%20Reference)))

(event EventTick
  (params (DeltaSeconds :type Float%20%28single-precision%29)))

(event EventOnDeathPlayEffects
  (params (Damage :type Float%20%28single-precision%29) (DamageTags :type Gameplay%20Tag%20Container%20Structure%20%28by%20ref%29) (Momentum :type Vector :default "0, 0, 0") (HitInfo :type Hit%20Result%20Structure%20%28by%20ref%29) (InstigatedBy :type Fort%20Pawn%20Object%20Reference) (DamageCauser :type Actor%20Object%20Reference) (EffectContext :type Gameplay%20Effect%20Context%20Handle%20Structure))
  (bind (_returnvalue_location _returnvalue_rotation) (Math|Transform|BreakTransform (Transformation|GetActorTransform)))
  (|CallParentFunction Damage DamageTags Momentum HitInfo InstigatedBy DamageCauser EffectContext)
  (if (TouchPlatform|IsMobileGame)
    (bind _returnvalue (Niagara|SpawnSystematLocation "/Game/Effects/Fort_Effects/World/Destruction/Destruction2/Trees/NS_Destruction_Toon_Trees.NS_Destruction_Toon_Trees" _returnvalue_location _returnvalue_rotation))
    (bind _returnvalue_2 (Class|FortFXCustomizationComponent|ApplyFXCustomizationtoComponent self _returnvalue))
    (Development|PrintString (Utilities|String|Append "Customization Applied:" (Utilities|String|ToString(Boolean) _returnvalue_2)))))

(fn GetConnectedOverlappingActors
  (params (Existing%20List :type Array%20of%20Actor%20Object%20References))
  (bind _localarraylist (LocalVariables|GetLocalArrayList))
  (for _array_element Existing%20List
    (Utilities|Array|AddUnique (LocalVariables|GetLocalArrayList) _array_element))
  (bind _outactors (Collision|ComponentOverlapActors (Variables|Debug|GetStaticMeshComponent) (Transformation|GetWorldTransform (Variables|Debug|GetStaticMeshComponent)) (Utilities|Array|MakeArray "ObjectTypeQuery1") "/WildEstate/Foliage/Hedges/Blueprints/BP_WildEstate_Hedge_Parent.BP_WildEstate_Hedge_Parent_C" (Utilities|Array|MakeArray self)))
  (for _array_element_2 _outactors
    (if (not (Utilities|Array|ContainsItem _localarraylist _array_element_2))
      (Utilities|Array|AddUnique _localarraylist _array_element_2)
      (bind _asbp_wild_estate_hedge_parent (Utilities|Casting|CastToBP_WildEstate_Hedge_Parent _array_element_2))
      (bind _actorarray (CallFunction|GetConnectedOverlappingActors _asbp_wild_estate_hedge_parent (LocalVariables|GetLocalArrayList)))
      (for _array_element_3 _actorarray
        (Utilities|Array|AddUnique (LocalVariables|GetLocalArrayList) _array_element_3))))
  (return (LocalVariables|GetLocalArrayList)))

(fn MakeRandomOutlineValueFromPosition
  (params (Position :type Vector))
  (bind _stream (LocalVariables|GetStream))
  (bind _returnvalue (Math|Random|RandomFloatinRangefromStream _stream 0.0 1.0))
  (Math|Random|SetRandomStreamSeed _stream (Math|Float|Truncate (+ (+ (.x Position) (.y Position)) (.z Position))))
  (CallFunction|SetOutlineValueandPrimData _returnvalue)
  (return _returnvalue))

(fn SetAllConnectedActorsArray
  (bind _outlinemergeconnectedactors_chain (Variables|Outline|GetOutlineMergeConnectedActors_Chain))
  (bind _outlinemergeconnectedactors_chain_2 (Variables|Outline|GetOutlineMergeConnectedActors_Chain))
  (Utilities|FlowControl|Sequence
    (:then_1
      (for _array_element (Variables|Outline|GetOutlineMergeConnectedActors_Chain)
        (bind _asbp_wild_estate_hedge_parent (Utilities|Casting|CastToBP_WildEstate_Hedge_Parent _array_element))
        (Variables|Outline|SetOutlineMergeConnectedActors_Chain (Variables|Outline|GetOutlineMergeConnectedActors_Chain) _asbp_wild_estate_hedge_parent)))
    (:then_2
      (if (> (Utilities|Array|Length _outlinemergeconnectedactors_chain_2) 0)
        (Utilities|IsValid (Utilities|Array|Get(acopy) _outlinemergeconnectedactors_chain_2)
          (:Is%20Valid
            (LocalVariables|SetMergedLocation "0, 0, 0")
            (for _array_element_2 _outlinemergeconnectedactors_chain
              (LocalVariables|SetMergedLocation (+ (LocalVariables|GetMergedLocation) (Transformation|GetActorLocation _array_element_2))))
            (LocalVariables|SetMergedLocation (/ (LocalVariables|GetMergedLocation) (Utilities|Array|Length _outlinemergeconnectedactors_chain))))
          (:Is%20Not%20Valid
            (LocalVariables|SetMergedLocation (Transformation|GetActorLocation))))
        (CallFunction|MakeRandomOutlineValuefromPosition (LocalVariables|GetMergedLocation))
        (for _array_element_3 (Variables|Outline|GetOutlineMergeConnectedActors_Chain)
          (bind _asbp_wild_estate_hedge_parent_2 (Utilities|Casting|CastToBP_WildEstate_Hedge_Parent _array_element_3))
          (CallFunction|SetOutlineValueandPrimData _asbp_wild_estate_hedge_parent_2 (Variables|Outline|GetOutlineMergeValue)))))))

(fn SetOutlineValueAndPrimData
  (params (OutlineValue :type Float%20%28double-precision%29))
  (bind _output_get (Variables|Outline|SetOutlineMergeValue OutlineValue))
  (CallFunction|SetRoughnesValuePrimData _output_get))

(fn SetRoughnesValuePrimData
  (params (0-1%20Roughnessvalue :type Float%20%28double-precision%29))
  (Rendering|Material|SetDefaultCustomPrimitiveDataFloat (Variables|Debug|GetStaticMeshComponent) 26 (Math|Float|Lerp 0.01 0.49 0-1%20Roughnessvalue)))

(fn SetSeasonPresetColors
  (bind _array_element (Utilities|Array|ForEachLoopwithBreak (Variables|/Game/ArtTools/Procedural/Trees/TreeSeasonColor_Data.TreeSeasonColor_Data_C|GetTreeSeasonData (Variables|LeaveColor|GetTreeSeasonColor%20Data)))
    (:LoopBody
      (if (Utilities|Enum|Equal(Enum) (Utilities|Struct|BreakTreeSeasonColorStruct _array_element) (Variables|LeaveColor|GetTree%20Color%20Option))
        (Variables|LeaveColor|SetColor (.Color_5_E33006414CF6D999D4D482A4D2FF03F6 (Utilities|Struct|BreakTreeSeasonColorStruct _array_element)))
        (Variables|LeaveColor|SetHue_Shift_BlueVertexColor (.ValueVariation_9_4FAED9E04DA8955B0BC4EDB1E0A3C1F7 (Utilities|Struct|BreakTreeSeasonColorStruct _array_element)))
        (Variables|LeaveColor|SetHueVariation_PerLeaf (.HueVariation_10_00BFFB9140C4AD7D16C9AA9AEB01A935 (Utilities|Struct|BreakTreeSeasonColorStruct _array_element)))))))

(fn SetTreeSeasonalColor
  (Utilities|FlowControl|Sequence
    (:then_0
      (if (Utilities|Enum|NotEqual(Enum) (Variables|LeaveColor|GetTree%20Color%20Option) "NewEnumerator0")
        (Default|SetSeasonPresetColors)))
    (:then_1
      (Default|UpdateColor))))

(fn UpdateColor
  (bind (_r _g _b _a) (Utilities|Struct|BreakLinearColor (Variables|LeaveColor|GetColor)))
  (bind (_r_2 _g_2 _b_2) (Utilities|Struct|BreakLinearColor (Variables|LeaveColor|GetColor_2)))
  (bind _staticmeshcomponent (Variables|Debug|GetStaticMeshComponent))
  (Rendering|Material|SetDefaultCustomPrimitiveDataVector4 _staticmeshcomponent 0 (Math|Vector4|MakeVector4 _r _g _b _a))
  (Rendering|Material|SetDefaultCustomPrimitiveDataVector3 _staticmeshcomponent 17 (Math|Vector|MakeVector _r_2 _g_2 _b_2))
  (Rendering|Material|SetDefaultCustomPrimitiveDataFloat (Variables|Debug|GetStaticMeshComponent) 4 (Variables|LeaveColor|GetHueVariation_PerLeaf))
  (Rendering|Material|SetDefaultCustomPrimitiveDataFloat (Variables|Debug|GetStaticMeshComponent) 5 (Variables|LeaveColor|GetHue_Shift_BlueVertexColor))
  (Rendering|Material|SetDefaultCustomPrimitiveDataFloat (Variables|Debug|GetStaticMeshComponent) 6 (Utilities|Select :Index (Variables|LeaveColor|GetTree%20Color%20Option) :NewEnumerator0 0.0 :NewEnumerator1 0.0 :NewEnumerator2 0.0 :NewEnumerator3 0.0 :NewEnumerator4 1.0)))

(fn ConstructionScript
  (bind _returnvalue (Transformation|GetActorTransform))
  (Utilities|FlowControl|Sequence
    (:then_0
      (Default|SetTreeSeasonalColor))
    (:then_1
      (if (Variables|Outline|GetUseOutlineMerging)
        (Utilities|FlowControl|Sequence
          (:then_0
            (if (not (== _returnvalue (Variables|Outline|GetStoredLocation)))
              (Variables|Outline|SetStoredLocation _returnvalue)
              (Utilities|FlowControl|Sequence
                (:then_0
                  (bind _actorarray (CallFunction|GetConnectedOverlappingActors))
                  (Variables|Outline|SetOutlineMergeConnectedActors_Chain _actorarray))
                (:then_1
                  (CallFunction|SetAllConnectedActorsArray)))))
          (:then_1
            (CallFunction|SetRoughnesValuePrimData (Variables|Outline|GetOutlineMergeValue))))
        (else
          (CallFunction|MakeRandomOutlineValuefromPosition (Transformation|GetActorLocation)))))))
