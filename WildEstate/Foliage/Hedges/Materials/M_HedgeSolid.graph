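# Masked, two-sided Nanite hedge material that reacts to an interaction position, with an IsLeaves switch for leaf cards.
# Distances to two height-offset points above that position create falloff masks, which world-aligned noise breaks up through HeightLerp for solid-hedge opacity.
# Leaves use a cutout texture and offsets that push them along the surface and normal, while ShadowReplace substitutes a camera-relative push in shadows.
# UV1 and UV2 decode leaf pivots.
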Material /WildEstate/Foliage/Hedges/Materials/M_HedgeSolid
graph {
 properties {bAutomaticallySetUsageInEditor=false, bUsedWithNanite=true, blendMode=BLEND_Masked, twoSided=true, usedShadingModels=Default Lit}
 nodes [
  Constant3Vector()
  Constant3Vector() {constant=<rgba 1 0 0 1>}
  Constant() {r=1}
  Constant()
  Constant() {r=1}
  n1 = WorldPosition()
  n2 = Constant3Vector() {desc=Axis, constant=<rgba 0 0 1 0>}
  n3 = VectorParameter() {parameterName=InteractionPositionWS, defaultValue=<rgba 4399.3037 -6449.0679 501.0538 1>}
  n4 = Add(A=Multiply(A=ScalarParameter() {defaultValue=49.3029, parameterName=1stHeight}, B=@n2), B=@n3.RGB)
  n5 = Add(A=Multiply(A=ScalarParameter() {defaultValue=125.8785, parameterName=2ndHeight}, B=@n2), B=@n3.RGB)
  n6 = Min(A=Distance(A=@n1.XYZ, B=@n4), B=Distance(A=@n5, B=@n1.XYZ))
  n7 = Subtract(A=@n6, B=ScalarParameter() {defaultValue=33.8834, parameterName=InteractOffset})
  n8 = ScalarParameter() {defaultValue=40, parameterName=InteractDistance}
  n9 = NamedRerouteDeclaration(None=Saturate(None=Divide(A=@n7, B=@n8))) {name=InteractionMask, nodeColor=<rgba 0.3359 0 1 1>}
  n10 = ScalarParameter() {defaultValue=0.0321, parameterName=MaskSub}
  n11 = MaterialFunctionCall(
   TextureObject=TextureObjectParameter() {parameterName=NoiseHeight, texture=/WildEstate/Environment/TestContent/T_WildEstate_Tree_Noise_M}
   TextureSize=ScalarParameter() {defaultValue=114.6011, parameterName=NoiseSize}
  ) {materialFunction=&WorldAlignedTexture}
  n12 = ComponentMask(None=@n11.XYZ Texture) {r=true}
  n13 = ComponentMask(None=@n11.XYZ Texture) {g=true}
  n14 = ScalarParameter() {defaultValue=2.4722, parameterName=Contrast}
  n15 = MaterialFunctionCall(Transition Phase=Subtract(A=@n9, B=Multiply(A=@n10, B=@n12)), Height Texture=@n13, Contrast=@n14) {materialFunction=&HeightLerp}
  LinearInterpolate(Alpha=Round(None=@n15.Alpha))
  n16 = VectorParameter() {parameterName=Color, defaultValue=<rgba 0.0061 0.1354 0.0559 1>}
  n17 = Divide(
   A=Subtract(A=@n9, B=ScalarParameter() {defaultValue=0.5678, parameterName=TextureBlend_Offset})
   B=ScalarParameter() {defaultValue=0.048, parameterName=TextureBlend_Fade}
  )
  LinearInterpolate(A=Constant3Vector(), B=@n16.RGB, Alpha=LinearInterpolate(A=ComponentMask(None=@n11.XYZ Texture) {b=true}, Alpha=Saturate(None=@n17)))
  n18 = Subtract(A=Add(A=@n13), B=ScalarParameter() {defaultValue=33.9694, parameterName=Outline})
  MaterialFunctionCall(Transition Phase=Saturate(None=Ceil(None=@n18)), Height Texture=@n13, Contrast=@n14) {materialFunction=&HeightLerp}
  n19 = NamedRerouteUsage()
  n20 = Subtract(A=@n19, B=@n5)
  n21 = Length(None=@n20) {is=class MaterialExpressionLength}
  n22 = Subtract(A=@n19, B=@n4)
  n23 = Length(None=@n22)
  Max(A=@n21, B=@n23)
  n24 = TextureObject() {texture=/WildEstate/Foliage/Hedges/Textures/T_WIldEstate_Hedge_Tiling}
  n25 = ScalarParameter() {defaultValue=137.7912, parameterName=TextureSize}
  n26 = WorldPosition()
  Multiply(
   A=LinearInterpolate(
    B=MaterialFunctionCall(TextureObject=@n24, TextureSize=@n25, WorldPosition=@n26.XYZ) {materialFunction=&WorldAlignedTexture}.XYZ Texture
    Alpha=@n9
   ) {constA=1}
   B=@n16.RGB
  )
  n27 = Min(A=@n21, B=@n23)
  n28 = Subtract(A=@n27, B=ScalarParameter() {defaultValue=20.8491, parameterName=InteractOffset_Leaves})
  n29 = ScalarParameter() {defaultValue=40, parameterName=InteractDistance_Leaves}
  NamedRerouteDeclaration(None=Saturate(None=Divide(A=@n28, B=@n29))) {name=InteractionMask_Leaves, nodeColor=<rgba 0.3359 0 1 1>}
  n30 = TextureCoordinate() {coordinateIndex=2}
  n31 = AppendVector(A=Frac(None=TextureCoordinate() {coordinateIndex=1}), B=ComponentMask(None=Frac(None=@n30)) {r=true})
  n32 = Constant3Vector() {constant=<rgba 0.5 0.5 0.5 0>}
  n33 = Subtract(A=Multiply(A=Subtract(A=@n31, B=@n32)) {constB=768}, B=Constant3Vector() {constant=<rgba 0 0 128 0>})
  NamedRerouteDeclaration(None=@n33) {name=Pivot_2024_LocalSpace, nodeColor=<rgba 0 1 0.9531 1>}
  NamedRerouteDeclaration(None=@n33) {name=LeafPivot Position LS, nodeColor=<rgba 0.0781 1 0 1>}
  n34 = Subtract(A=@n27, B=ScalarParameter() {defaultValue=40, parameterName=InteractOffset_Leaves2})
  n35 = ScalarParameter() {defaultValue=0.1, parameterName=InteractDistance_Leaves2}
  NamedRerouteDeclaration(None=Saturate(None=Divide(A=@n34, B=@n35))) {name=InteractionMask_Leaves_Hard, nodeColor=<rgba 0.3359 0 1 1>}
  NamedRerouteDeclaration(None=TransformPosition(None=@n33) {transformSourceType=TRANSFORMPOSSOURCE_Instance, transformType=TRANSFORMPOSSOURCE_World}) {name=Pivot_2024_WorldSpace, nodeColor=<rgba 0 1 0.9531 1>}
  NamedRerouteDeclaration(None=TransformPosition(None=@n33) {transformSourceType=TRANSFORMPOSSOURCE_Instance, transformType=TRANSFORMPOSSOURCE_TranslatedWorld}) {name=Pivot_2024_CameraSpace, nodeColor=<rgba 0 1 0.9531 1>}
  n36 = Subtract(A=@n6, B=ScalarParameter() {defaultValue=30, parameterName=InteractOffset_LeavSoft})
  n37 = ScalarParameter() {defaultValue=15, parameterName=InteractDistance_LeafSoft}
  NamedRerouteDeclaration(None=Saturate(None=Divide(A=@n36, B=@n37))) {name=InteractionMask_LeafsSoft, nodeColor=<rgba 0.3359 0 1 1>}
  NamedRerouteDeclaration(None=ComponentMask(None=@n30) {g=true}) {name=Debug_Float, nodeColor=<rgba 0.8516 1 0 1>}
  NamedRerouteDeclaration(None=@n33) {name=Debug_Vector, nodeColor=<rgba 0.8281 0 1 1>}
  NamedRerouteUsage()
  NamedRerouteUsage()
  NamedRerouteUsage()
  Round(None=Subtract(A=@n15.Alpha, B=ScalarParameter() {defaultValue=0.1, parameterName=Dist}))
  Saturate(None=Divide(B=ScalarParameter() {defaultValue=33.7981, parameterName=DepthInterSection}))
  ScalarParameter() {defaultValue=0.5, parameterName=SSS_ShadowOffset}
  n38 = Normalize(VectorInput=@n20)
  n39 = Normalize(VectorInput=@n22)
  n40 = Normalize(VectorInput=Multiply(A=Constant3Vector() {constant=<rgba 1 1 0 0>}, B=Add(A=@n38, B=@n39)))
  n41 = VertexNormalWS()
  n42 = CrossProduct(A=CrossProduct(A=@n41, B=NamedRerouteDeclaration(None=@n40) {name=Interaction_Vector, nodeColor=<rgba 0.3359 0 1 1>}), B=@n41)
  n43 = Sign(None=DotProduct(A=@n40, B=@n42))
  n44 = Multiply(
   A=Normalize(VectorInput=Multiply(A=@n43, B=@n42))
   B=LinearInterpolate(A=ScalarParameter() {defaultValue=40, parameterName=LeavesPush}, Alpha=NamedRerouteUsage()) {constB=0}
  )
  n45 = Multiply(
   A=VertexNormalWS()
   B=LinearInterpolate(A=ScalarParameter() {defaultValue=10, parameterName=LeavesPushForward}, Alpha=NamedRerouteUsage()) {constB=0}
  )
  n46 = WorldPosition() {worldPositionShaderOffset=WPT_CameraRelative}
  ShadowReplace(
   Default=StaticSwitchParameter(True=Add(A=@n44, B=@n45), False=Constant3Vector()) {parameterName=IsLeaves}
   Shadow=Multiply(A=Normalize(VectorInput=@n46.XYZ), B=ScalarParameter() {defaultValue=8, parameterName=ShadowPush})
  )
  n47 = Constant()
  SingleLayerWaterMaterialOutput!(ScatteringCoefficients=@n47, AbsorptionCoefficients=@n47, PhaseG, ColorScaleBehindWater)
  StaticSwitchParameter(
   True=TextureSample() {texture=/WildEstate/Foliage/Hedges/Meshes/T_WIldEstate_Hedge_Leave_CutOut}.RGB
   False=LinearInterpolate(A=ScalarParameter() {parameterName=OpacityMask}, Alpha=@n15.Alpha) {constA=1}
  ) {parameterName=IsLeaves}
  Subtract(A=@n1.XYZ, B=@n4)
  TextureSampleParameter2D() {group=Textures, parameterName=NoiseHeight, texture=&T_WildEstate_Tree_Noise_M}
  TextureSample() {texture=&T_WIldEstate_Hedge_Tiling}
  TransformPosition(None=@n33) {transformType=TRANSFORMPOSSOURCE_World}
 ]
}
