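# Test masked subsurface material for WildEstate hedges, using a tiled tree noise texture's green channel as base color.
# HeightLerp combines that noise with biased red vertex color to produce the opacity mask.
# The world-space normal is the normalized sum of sunlight direction and camera vector, noted as a compromise between directional and other lights.
# Fresnel-like, sun-shaded color and alternative mask-bias branches are left unconnected.
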Material /WildEstate/Foliage/Hedges/Materials/M_WildEstate_HedgeTestMasked
graph {
 properties {
  bAutomaticallySetUsageInEditor=false
  bTangentSpaceNormal=false
  bUseMaterialAttributes=true
  blendMode=BLEND_Masked
  shadingModel=MSM_Subsurface
  usedShadingModels=Subsurface
 }
 nodes [
  n1 = TextureSample(UVs=Multiply(A=TextureCoordinate(), B=ScalarParameter() {defaultValue=2, parameterName=Tiling})) {texture=/WildEstate/Environment/TestContent/T_WildEstate_Tree_Noise_M}
  Add(A=DotProduct(A=VertexNormalWS(), B=CameraVectorWS()), B=@n1.G)
  n2 = VertexColor()
  n3 = Multiply(A=@n2.R, B=ScalarParameter() {defaultValue=1.0106, parameterName=MaskedBiasVTX})
  n4 = Subtract(A=@n1.G, B=@n3)
  Ceil(None=@n4)
  n5 = CollectionParameter() {
   collection=&FortniteMaterialParameters
   parameterName=SunlightDirectionalVector
  }
  n6 = Add(A=DotProduct(A=VertexNormalWS(), B=ComponentMask(None=@n5) {b=true, g=true, r=true}))
  Multiply(
   A=LinearInterpolate(A=ScalarParameter() {defaultValue=0.104, parameterName=MinShadowBrightness}, Alpha=Divide(A=@n6))
   B=VectorParameter() {parameterName=Color, defaultValue=<rgba 0.0061 0.1354 0.0559 1>}.RGB
  )
  OneMinus(None=@n2.R)
  n7 = CollectionParameter() {
   collection=&FortniteMaterialParameters
   parameterName=SunlightDirectionalVector
  }
  n8 = ComponentMask(None=@n7) {b=true, g=true, r=true}
  n9 = CameraVectorWS()
  SetMaterialAttributes!(
   MaterialAttributes
   Base Color=@n1.G
   Specular=Constant()
   Roughness=Constant() {r=1}
   Opacity=ScalarParameter() {parameterName=Opacity}
   Subsurface Color=Constant3Vector()
   Normal=Normalize(VectorInput=Add(A=@n8, B=@n9) {desc=Compromise for Direction light and all othe rlights})
   Opacity Mask=MaterialFunctionCall(Transition Phase=@n3, Height Texture=@n1.G) {materialFunction=&HeightLerp}.Alpha
  ) {attributeSetTypes [
    #1
    #2
    #3
    #4
    #5
    #6
    #7
   ]}
  Subtract(A=@n4, B=ScalarParameter() {defaultValue=0.595, parameterName=MaskedBias})
 ]
}
