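; Test actor for previewing the hedge material's interaction effect against a Jonesy agent mesh.
; The construction script moves Jonesy to DebugPosition and puts a dynamic instance of M_WildEstate_Hedge_A on the hedge mesh.
; It feeds InteractionPositionWS with actor location plus DebugPosition and a fixed upward offset, ignoring actor rotation.
; BeginPlay, overlap and Tick events are empty, so nothing updates at runtime.
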(event EventBeginPlay)

(event Collision|EventActorBeginOverlap
  (params (OtherActor :type Actor%20Object%20Reference)))

(event EventTick
  (params (DeltaSeconds :type Float%20%28single-precision%29)))

(fn ConstructionScript
  (bind _debugposition (Variables|Default|GetDebugPosition))
  (bind _sm_wildestate_hedge_b (Variables|Default|GetSM_WildEstate_Hedge_B))
  (Transformation|SetRelativeLocation (Variables|Default|GetM_MED_Agent_Jonesy) _debugposition false false)
  (bind _returnvalue (Rendering|Material|CreateDynamicMaterialInstance _sm_wildestate_hedge_b 0 "/WildEstate/Foliage/Hedges/Materials/M_WildEstate_Hedge_A.M_WildEstate_Hedge_A" "None" :!class /Script/Engine.PrimitiveComponent))
  (Rendering|Material|SetMaterial _sm_wildestate_hedge_b 0 _returnvalue)
  (Rendering|Material|SetVectorParameterValue _returnvalue "InteractionPositionWS" (Math|Conversions|ToLinearColor(Vector) (+ (Transformation|GetActorLocation) (+ _debugposition "0.000000,0.000000,88.000000"))) :!class /Script/Engine.MaterialInstanceDynamic))
