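; Test actor for hedge interaction, pairing a Wild Estate hedge mesh with a Jonesy agent component that stands in as the interactor.
; The construction script gives each hedge material a dynamic instance, seeds its InteractionPositionWS parameter and caches it in MatIDs.
; Each Tick, a sine of game time slides Jonesy between two fixed local points and pushes the actor location plus that offset to the cached materials.
; A custom Update Point event duplicates the Tick logic.
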(event EventBeginPlay)

(event Collision|EventActorBeginOverlap
  (params (OtherActor :type Actor%20Object%20Reference)))

(event EventTick
  (params (DeltaSeconds :type Float%20%28single-precision%29))
  (bind _debugposition (Variables|Default|GetDebugPosition))
  (Variables|Default|SetDebugPosition (Math|Vector|Lerp(Vector) "-14.696012,86.382939,15.168496" "-139.316336,86.382939,15.168496" (/ (+ (Math|Trig|Sin(Radians) (* (Utilities|Time|GetGameTimeinSeconds) (Variables|Default|GetSpeed))) 1.0) 2.0)))
  (Transformation|SetRelativeLocation (Variables|Default|GetM_MED_Agent_Jonesy) _debugposition false false)
  (for _array_element (Variables|Default|GetMatIDs)
    (Rendering|Material|SetVectorParameterValue _array_element "InteractionPositionWS" (Math|Conversions|ToLinearColor(Vector) (+ (Transformation|GetActorLocation) _debugposition)) :!class /Script/Engine.MaterialInstanceDynamic)))

(event Custom|Update%20Point
  (bind _debugposition (Variables|Default|GetDebugPosition))
  (Variables|Default|SetDebugPosition (Math|Vector|Lerp(Vector) "-14.696012,86.382939,15.168496" "-139.316336,86.382939,15.168496" (/ (+ (Math|Trig|Sin(Radians) (* (Utilities|Time|GetGameTimeinSeconds) (Variables|Default|GetSpeed))) 1.0) 2.0)))
  (Transformation|SetRelativeLocation (Variables|Default|GetM_MED_Agent_Jonesy) _debugposition false false)
  (for _array_element (Variables|Default|GetMatIDs)
    (Rendering|Material|SetVectorParameterValue _array_element "InteractionPositionWS" (Math|Conversions|ToLinearColor(Vector) (+ (Transformation|GetActorLocation) _debugposition)) :!class /Script/Engine.MaterialInstanceDynamic)))

(fn ConstructionScript
  (bind _debugposition (Variables|Default|GetDebugPosition))
  (bind _sm_wildestate_hedge_b (Variables|Default|GetSM_WildEstate_Hedge_B))
  (Transformation|SetRelativeLocation (Variables|Default|GetM_MED_Agent_Jonesy) _debugposition false false)
  (for (_array_element _array_index) (Rendering|Material|GetMaterials _sm_wildestate_hedge_b)
    (bind _returnvalue (Rendering|Material|CreateDynamicMaterialInstance _array_element "None" "None" :!class /Script/Engine.KismetMaterialLibrary))
    (Rendering|Material|SetMaterial _sm_wildestate_hedge_b _array_index _returnvalue)
    (Rendering|Material|SetVectorParameterValue _returnvalue "InteractionPositionWS" (Math|Conversions|ToLinearColor(Vector) (+ (Transformation|GetActorLocation) _debugposition)) :!class /Script/Engine.MaterialInstanceDynamic)
    (Utilities|Array|Add (Variables|Default|GetMatIDs) _returnvalue)))
