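# TestHedgeLoop is a Level Sequence at 30 fps, apparently a test for the hedge foliage Blueprint BP_HedgeTest2.
# It possesses a BP_HedgeTest2 actor and gives that binding a single event track with a trigger section.
# The event entry names no function, and the sequence has no other tracks, so what it fires or loops is not shown.
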LevelSequence /WildEstate/Foliage/Hedges/Meshes/TestHedgeLoop
bindings [BP_HedgeTest2 {
  id=#1
  possessed_class=BP_HedgeTest2_C
  tracks [{
    type=MovieSceneEventTrack
    sections [{
      type=MovieSceneEventTriggerSection
      channels [None {type=MovieSceneScriptingEventChannel, const ["<Struct 'MovieSceneEvent' (0x8537d34ba40) {}>", 1]}]
     }]
   }]
 }]
display_rate {denominator=1, numerator=30}
tick_resolution {denominator=1, numerator=24000}
tracks []
